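# Material function for the Voltaic Huntress accessory that outputs an animated mask of lines following an SDF texture's contours.
# Scaled depth UVs rotate over offset time and are distorted by mask-texture noise panned in radial coordinates, then sample the SDF.
# A time-scrolled sine of the SDF value forms the moving lines.
# A circular falloff, with front and back distance masks and a UseFrontMask static switch, fades the result.
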MaterialFunction /BRCosmetics/Characters/Player/Female/Medium/Bodies/F_MED_VoltaicHuntress/Materials/MaterialFunctions/MF_VoltaicHuntress_Accessory_SDF
graph {nodes [
  n1 = TextureCoordinate()
  n2 = NamedRerouteDeclaration(None=MaterialFunctionCall(
    UVs=FunctionInput(Preview=@n1) {inputName=DepthUV, inputType=FunctionInput_Vector2}
    Texture Scale=ScalarParameter() {defaultValue=0.1, group=VHAccessory, parameterName=CircleEffectScale}
   ) {materialFunction=&ScaleUVsByCenter}.UVs) {name=DepthUV, nodeColor=<rgba 0 1 0.625 1>}
  n3 = Constant()
  n4 = Add(
   A=Time()
   B=NamedRerouteDeclaration(None=FunctionInput(Preview=@n3) {bUsePreviewValueAsDefault=true, inputName=TimeOffset, inputType=FunctionInput_Scalar}) {name=TimeOffset, nodeColor=<rgba 1 0.2578 0 1>}
  )
  n5 = MaterialFunctionCall(Vector or UVs=Add(A=NamedRerouteUsage()) {constB=-0.5}) {materialFunction=&VectorToRadialValue}
  n6 = TextureCoordinate()
  n7 = TextureSample!(
   UVs=Panner(Coordinate=@n5.Radial Coordinates) {speedX=0.02, speedY=-0.3}
   Tex
   "DDX(UVs)"=DDX(Value=@n6)
   "DDY(UVs)"=DDY(Value=@n6)
   Apply View MipBias
  ) {
   mipValueMode=TMVM_Derivative
   samplerSource=SSM_Wrap_WorldGroupSettings
   samplerType=SAMPLERTYPE_LinearColor
   texture=/BRCosmetics/Characters/Player/Female/Medium/Bodies/F_MED_VoltaicHuntress/Materials/T_VoltaicHuntress_Mask
  }
  n8 = Subtract(A=@n7.R) {constB=0.5}
  n9 = TextureSample!(UVs=Add(A=Rotator(Coordinate=@n2, Time=@n4), B=Multiply(A=@n8) {constB=0.075}), Tex, Apply View MipBias) {
   samplerSource=SSM_Clamp_WorldGroupSettings
   samplerType=SAMPLERTYPE_LinearGrayscale
   texture=/BRCosmetics/Characters/Player/Female/Medium/Bodies/F_MED_VoltaicHuntress/Materials/T_VoltaicHuntress_SDF
  }
  n10 = Ceil(None=@n9.R)
  n11 = Add(A=Time(), B=Divide(A=NamedRerouteUsage()) {constB=10})
  n12 = ScalarParameter() {defaultValue=1.5, group=VHAccessory, parameterName=LinesSpeed}
  n13 = Sine(None=Add(A=Multiply(A=@n9.R) {constB=5.5}, B=Multiply(A=@n11, B=@n12) {constB=1.3}))
  n14 = StaticSwitchParameter(True=NamedRerouteUsage(), False=NamedRerouteUsage()) {defaultValue=true, group=VHAccessory, parameterName=UseFrontMask}
  n15 = Multiply(
   A=Power(Base=OneMinus(None=@n14), Exp=ScalarParameter() {defaultValue=7, group=VHAccessory, parameterName=CircleMaskSmoothness}) {constExponent=2.5}
   B=ScalarParameter() {defaultValue=1, group=VHAccessory, parameterName=CircleMaskIntensity}
  )
  n16 = Multiply(A=SmoothStep(Value=Multiply(A=@n10, B=@n13)) {constMin=0.6}, B=SmoothStep(Value=Saturate(None=@n15)))
  n17 = Constant() {r=2}
  FunctionOutput(None=Saturate(None=Multiply(A=@n16, B=@n17)))
  n18 = TextureCoordinate()
  n19 = Multiply(A=Subtract(A=FunctionInput(Preview=@n18) {inputName=UVs, inputType=FunctionInput_Vector2}, B=Constant() {r=0.5})) {constB=10}
  n20 = Multiply(A=@n19, B=@n19)
  n21 = ComponentMask(None=@n20) {r=true}
  n22 = ComponentMask(None=@n20) {g=true}
  NamedRerouteDeclaration(None=SquareRoot(None=Add(A=@n21, B=@n22))) {name=FrontFadeMask, nodeColor=<rgba 0 1 0.4375 1>}
  n23 = Multiply(A=Subtract(A=FunctionInput() {inputName=CircleCenterPosition, inputType=FunctionInput_Vector2}, B=Constant() {r=0.5})) {constB=10}
  n24 = Multiply(A=@n23, B=@n23)
  n25 = ComponentMask(None=@n24) {r=true}
  n26 = ComponentMask(None=@n24) {g=true}
  NamedRerouteDeclaration(None=SquareRoot(None=Add(A=@n25, B=@n26))) {name=BackFadeMask, nodeColor=<rgba 0 1 0.4375 1>}
  NamedRerouteDeclaration(None=@n5.Linear Distance) {name=RadialGradient, nodeColor=<rgba 0 0.3672 1 1>}
 ]}
