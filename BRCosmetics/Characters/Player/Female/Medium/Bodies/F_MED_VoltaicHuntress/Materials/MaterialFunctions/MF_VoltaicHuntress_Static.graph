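# Material function that outputs an animated static mask for the Voltaic Huntress cosmetic character, with parameters grouped under VHCapsule.
# It rotates and scales UVs, adds BumpOffset parallax from a Depth input, then pans them over time with an offset.
# A panned red-channel mask sample distorts the V coordinate of a second sample, whose alpha is shaped by Power and SmoothStep.
# The flicker and rounded-time branches do not reach the output.
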MaterialFunction /BRCosmetics/Characters/Player/Female/Medium/Bodies/F_MED_VoltaicHuntress/Materials/MaterialFunctions/MF_VoltaicHuntress_Static
graph {nodes [
  n1 = Constant() {r=0.1}
  Add(
   A=Constant() {r=0.125}
   B=Multiply(A=MaterialFunctionCall!(FlickerInitialSpeed=@n1, FlickerRandom=@n1) {materialFunction=/Game/Effects/Fort_Effects/Materials/_Functions/MF_Flicker}.Result) {constB=0.1}
  )
  n2 = MaterialFunctionCall(
   UVs=TextureCoordinate()
   "Rotation Angle (0-1)"=FunctionInput(Preview=ScalarParameter() {defaultValue=0.25, group=VHCapsule, parameterName=StaticRotation}) {bUsePreviewValueAsDefault=true, inputName="Rotation(0-1)", inputType=FunctionInput_Scalar}
  ) {materialFunction=&CustomRotator}
  n3 = FunctionInput(Preview=Constant() {r=10}) {bUsePreviewValueAsDefault=true, inputName="Scale(10)", inputType=FunctionInput_Vector2}
  n4 = Multiply(A=@n2.Rotated Values, B=@n3) {constB=15}
  n5 = FunctionInput(Preview=Constant()) {bUsePreviewValueAsDefault=true, inputName=Depth, inputType=FunctionInput_Scalar}
  n6 = FunctionInput() {bUsePreviewValueAsDefault=true, inputName=TimeOffset, inputType=FunctionInput_Scalar}
  n7 = NamedRerouteUsage()
  n8 = Panner(Coordinate=Add(A=BumpOffset(Coordinate=@n4, Height=@n5) {heightRatio=0.04}, B=@n6), Time=@n7) {speedX=-0.2, speedY=0.01}
  n9 = ComponentMask(None=@n8) {g=true}
  n10 = Multiply(A=@n2.Rotated Values, B=Divide(A=@n3) {constB=3}) {constB=7}
  n11 = NamedRerouteDeclaration(None=Add(A=Time(), B=@n6)) {name=TimeOffset, nodeColor=<rgba 1 0 0.1875 1>}
  n12 = TextureSample(UVs=Panner(Coordinate=BumpOffset(Coordinate=@n10, Height=@n5) {heightRatio=0.04}, Time=@n11) {speedX=0.05, speedY=0.025}) {
   samplerType=SAMPLERTYPE_LinearColor
   texture=/BRCosmetics/Characters/Player/Female/Medium/Bodies/F_MED_VoltaicHuntress/Materials/T_VoltaicHuntress_Mask
  }
  n13 = Multiply(A=Subtract(A=@n12.R) {constB=0.5}) {constB=0.1}
  n14 = TextureSample(UVs=AppendVector(A=ComponentMask(None=@n8) {r=true}, B=Add(A=@n9, B=@n13))) {
   samplerType=SAMPLERTYPE_LinearColor
   texture=&T_VoltaicHuntress_Mask
  }
  n15 = ScalarParameter() {defaultValue=1, group=VHCapsule, parameterName=StaticPower}
  FunctionOutput(None=SmoothStep(
    Min=ScalarParameter() {defaultValue=0.01, group=VHCapsule, parameterName=StaticThickness}
    Value=Power(Base=@n14.A, Exp=@n15) {constExponent=1}
   ) {constMin=0.7}) {outputName=Static}
  Round(None=Multiply(A=@n7, B=ScalarParameter() {defaultValue=5, group=VHCapsule, parameterName=StaticPanningSpeed}) {constB=3.5})
  Round(None=Multiply(A=@n11, B=ScalarParameter() {defaultValue=10, group=VHCapsule, parameterName=StaticDeformationSpeed}) {constB=10})
 ]}
